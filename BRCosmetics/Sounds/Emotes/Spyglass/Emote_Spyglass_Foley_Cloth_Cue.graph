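# Sound cue for cloth foley in the Spyglass emote.
# A random node picks one of three equally weighted Commando stereo foley waves, played at reduced volume.
# It uses emote attenuation, the emote foley sound class and property sheet, and a shared emote foley concurrency set.
# It also references the music audio analysis submix as a send.
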SoundCue /BRCosmetics/Sounds/Emotes/Spyglass/Emote_Spyglass_Foley_Cloth_Cue
root=Random_0
nodes [
 {id=Random_0, children [WavePlayer_0,WavePlayer_5,WavePlayer_6], properties {weights [1,1,1]}}
 {id=WavePlayer_0, properties {soundWaveAssetPtr=/Game/Sounds/Fortress_Foley/Commando/Commando_Foley_Stereo_007}}
 {id=WavePlayer_5, properties {soundWaveAssetPtr=/BRCosmetics/Sounds/Other/Fortress_Foley/Commando/Commando_Foley_Stereo_010}}
 {id=WavePlayer_6, properties {soundWaveAssetPtr=/Game/Sounds/Fortress_Foley/Commando/Commando_Foley_Stereo_003}}
]
properties {
 attenuationSettings=&Fort_Emotes_Athena_Att
 audioPropertiesSheet=/BRCosmetics/Sounds/Settings/PropertySheets/Emotes/APS_Cosmetic_Emote_Foley
 maxDistance=1600
 soundClassObject=/Game/Sounds/FortSoundClassesAndModes/Fort_Foley_Emote
 volumeMultiplier=0.2
 concurrencyOverrides {maxCount=3, resolutionRule=StopOldest}
 concurrencySet [/Game/Sounds/Concurrency_Rules/Athena_EmoteFoley_Conc]
 soundSubmixSends [{minSendLevel=0, soundSubmix=&Fortnite_AudioAnalysis_Music}]
}
